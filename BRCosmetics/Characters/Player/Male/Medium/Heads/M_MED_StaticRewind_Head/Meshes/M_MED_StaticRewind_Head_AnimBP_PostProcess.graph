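# Post-process AnimBP for the male medium StaticRewind head cosmetic, adding secondary hair motion to the pose copied from the attached parent mesh.
# An additive head Transform Bone feeds four chained Anim Dynamics nodes, one per dyn_hair bone, using angular springs and limits.
# The event graph derives HeadLimit from targeting and HeadRot from negated, per-axis-scaled head-to-neck_02 rotation delta, but the head node shows zero rotation and constant alpha.
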AnimBlueprint /BRCosmetics/Characters/Player/Male/Medium/Heads/M_MED_StaticRewind_Head/Meshes/M_MED_StaticRewind_Head_AnimBP_PostProcess
target_skeleton=/Game/Characters/Player/Common/Fortnite_Base_Head/Fortnite_Base_Head_Export_Skeleton
graphs [
 AnimGraph {nodes [
   Input Pose()
   n1 = "Transform (Modify) Bone - Bone: head"!(
    ComponentPose=Local To Component(LocalPose=Copy Pose From Mesh() {bCopyCurves=true, bUseAttachedParent=true})
    bAlphaBoolEnabled:bool=True
    Translation="0,0,0"
    Rotation="0,0,0"
    Scale="1,1,1"
   ) -> (Pose) {
    alphaInputType=Bool
    rotationMode=BMM_Additive
    rotationSpace=BCS_BoneSpace
    alphaBoolBlend {blendInTime=0.3, blendOption=Linear, blendOutTime=0.3}
    boneToModify {boneName=head}
   }
   n2 = Anim Dynamics - Bone: dyn_hair_01(ComponentPose=@n1, Alpha=1) {
    angularSpringConstant=200
    bAngularSpring=true
    bUseGravityOverride=true
    alphaBoolBlend {blendOption=Linear}
    boundBone {boneName=dyn_hair_01}
    componentAppliedLinearAccClamp=<500 500 500>
    componentLinearAccScale=<0.2 0.2 0.2>
    componentLinearVelScale=<0.2 0.2 0.2>
    gravityOverride {z=-400}
    physicsBodyDefinitions [{
      boundBone {boneName=dyn_hair_01}
      boxExtents=<8.5 10 10>
      constraintSetup {angularLimitsMax=<0 0 50>, angularLimitsMin=<0 0 -30>, angularTarget=<0 0 0>}
      localJointOffset=<2.5 0 0>
     }]
    relativeSpaceBone {boneName=spine_05}
   }
   n3 = Anim Dynamics - Bone: dyn_hair_02(ComponentPose=@n2, Alpha=1) {
    angularDampingOverride=0.999
    angularSpringConstant=3
    bAngularSpring=true
    bOverrideAngularDamping=true
    bUseGravityOverride=true
    alphaBoolBlend {blendOption=Linear}
    boundBone {boneName=dyn_hair_02}
    componentAppliedLinearAccClamp=<600 600 600>
    componentLinearAccScale=<0.15 0.15 0.2>
    componentLinearVelScale=<0.15 0.15 0.1>
    gravityOverride {z=-600}
    physicsBodyDefinitions [{
      boundBone {boneName=dyn_hair_02}
      boxExtents=<5 10 10>
      constraintSetup {angularLimitsMax=<0 20 45>, angularLimitsMin=<0 -20 -20>, angularTarget=<0 0 0>}
      localJointOffset=<2 0 0>
     }]
   }
   n4 = Anim Dynamics - Bone: dyn_hair_03(ComponentPose=@n3, Alpha=1, ExternalForce="0,0,0") {
    angularDampingOverride=0.999
    angularSpringConstant=3
    bAngularSpring=true
    bOverrideAngularDamping=true
    bUseGravityOverride=true
    alphaBoolBlend {blendOption=Linear}
    boundBone {boneName=dyn_hair_03}
    componentAppliedLinearAccClamp=<800 800 1000>
    componentLinearAccScale=<0.15 0.15 0.2>
    componentLinearVelScale=<0.15 0.15 0.2>
    gravityOverride {z=-600}
    physicsBodyDefinitions [{
      boundBone {boneName=dyn_hair_03}
      boxExtents=<5 10 10>
      constraintSetup {angularLimitsMax=<0 10 40>, angularLimitsMin=<0 -10 -20>, angularTarget=<0 0 0>}
      localJointOffset=<2 0 0>
     }]
   }
   n5 = Anim Dynamics - Bone: dyn_hair_04(ComponentPose=@n4, Alpha=1, ExternalForce="0,0,0") {
    angularSpringConstant=3
    bAngularSpring=true
    bUseGravityOverride=true
    simulationSpace=BoneRelative
    alphaBoolBlend {blendOption=Linear}
    boundBone {boneName=dyn_hair_04}
    componentAppliedLinearAccClamp=<100 100 100>
    componentLinearAccScale=<0.1 0.1 0.1>
    componentLinearVelScale=<0.1 0.1 0.1>
    physicsBodyDefinitions [{
      boundBone {boneName=dyn_hair_04}
      boxExtents=<5 10 10>
      constraintSetup {angularLimitsMax=<0 5 5>, angularLimitsMin=<0 -5 -5>, angularTarget=<0 0 0>}
      localJointOffset=<5 0 0>
     }]
    relativeSpaceBone {boneName=spine_05}
   }
   Output Pose(Result=Component To Local(ComponentPose=@n5))
  ]}
 EventGraph {nodes [
   Event BlueprintUpdateAnimation()
   > Set HeadLimit(HeadLimit:bool=Get bIsTargeting()) -> (Output_Get:bool)
   n1 = BreakTransform!(InTransform=GetDeltaTransformFromRefPose(self=GetOwnerSkeletalMeshComponent(), BoneName=head, BaseName=neck_02)) -> (Location, Rotation_Roll:float, Rotation_Pitch:float, Rotation_Yaw:float, Scale)
   > Set!(
    execute
    HeadRot_Roll:float=float * float!(A=@n1.Rotation_Roll, B=-1, C:double=Get XPercentage() -> (XPercentage:double)) -> (ReturnValue) {is=title K2Node_PromotableOperator}
    HeadRot_Pitch:float=float * float(A=@n1.Rotation_Pitch, B=-1, C=Get YPercentage() -> (YPercentage:double))
    HeadRot_Yaw:float=float * float(A=@n1.Rotation_Yaw, B=-1, C=Get ZPercentage() -> (ZPercentage:double))
   ) -> (then, Output_Get:Rotator)
   TryGetPawnOwner()
  ]}
]
variables [
 CinePosing {type=boolean}
 HeadLimit {type=boolean}
 HeadRot {type=Rotator}
 XPercentage {type=number, default=1}
 YPercentage {type=number, default=1}
 ZPercentage {type=number, default=1}
]
